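# Unlit modulate material for Niagara mesh particles that tints the surface behind them with footprint-shaped marks for snow and sand interaction.
# It projects world position behind translucency into the particle's quaternion-rotated, scaled local space, then flips and rotates those UVs to sample a footprint mask texture.
# The mask's red channel blends a scaled tint toward white, MarkFlat fades the effect, and a custom node clips on green.
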Material /Game/Effects/SnowSandInteraction/Materials/ColorEffector/SSI_ColorEffector
graph {
 properties {
  bUseTranslucencyVertexFog=false
  bUsedWithNiagaraMeshParticles=true
  blendMode=BLEND_Modulate
  pixelDepthOffsetMode=PDOM_Legacy
  shadingModel=MSM_Unlit
  translucencyPass=MTP_BeforeDOF
  usedShadingModels=Unlit
 }
 nodes [
  n1 = MaterialFunctionCall() {materialFunction=&WorldPositionBehindTranslucency}
  n2 = LinearInterpolate(A=ParticlePositionWS(), B=ActorPositionWS(), Alpha=ScalarParameter() {parameterName=UseDebugPosition})
  n3 = DynamicParameter() {defaultValue=<rgba 1 1 0 0>, paramNames [CompressedQuaternion, BlurMarkScale, MarkFlat, MeshScale]}
  n4 = DynamicParameter() {parameterIndex=3, defaultValue=<rgba 1 1 0 0>, paramNames [QuatX,QuatY,QuatZ,QuatW]}
  n5 = MaterialFunctionCall!(
   Vector=Divide(A=Subtract(A=@n1.Result, B=@n2), B=Multiply(A=@n3.MeshScale) {constB=100}) {constB=100}
   Quat=MaterialFunctionCall(R=@n4.QuatX, G=@n4.QuatX, B=@n4.QuatX, A=@n4.QuatX) {materialFunction=&AppendMany}.RGBA
   Invert Quaternion=StaticBool() {value=true}
  ) {materialFunction=/Engine/Functions/Engine_MaterialFunctions03/Math/MultiplyVectorWithQuaternion}
  n6 = Add(A=ComponentMask(None=Add(A=@n5.Result) {constB=0.5}) {g=true, r=true}, B=Constant2Vector() {g=-0.05})
  n7 = DynamicParameter() {parameterIndex=1, defaultValue=<rgba 1 1 0 0>, paramNames [R, "", "", ""]}
  n8 = LinearInterpolate(A=Constant2Vector() {g=1, r=1}, B=Constant2Vector() {g=1, r=-1}, Alpha=@n7.G)
  n9 = MaterialFunctionCall(
   UVs=MaterialFunctionCall(
    UVs=MaterialFunctionCall(UVs=@n6, Texture Scale=@n8) {materialFunction=&ScaleUVsByCenter}.UVs
    Rotation Center=Constant2Vector() {g=0.5, r=0.5}
    "Rotation Angle (0-1)"=@n7.R
   ) {materialFunction=&CustomRotator}.Rotated Values
   Texture Scale=ScalarParameter() {defaultValue=1, parameterName=POM_UVScale}
  ) {materialFunction=&ScaleUVsByCenter}
  n10 = MaterialFunctionCall!(TexCoord=@n9.UVs, ClampChannel=Constant() {r=1}) {materialFunction=/Game/Effects/Fort_Effects/MaterialFunctions/MF_FX_TilingMethod}
  MaterialExpressionTextureSampleParameter2D_0 = TextureSampleParameter2D(UVs=MaterialFunctionCall(UVs=LinearInterpolate(B=@n9.UVs, Alpha=@n10.Result), Texture Scale=ScalarParameter() {defaultValue=0.6667}) {materialFunction=&ScaleUVsByCenter}.UVs) {parameterName=MaskTexture, texture=/Game/Effects/SnowSandInteraction/Textures/SSI_T_FootprintMasks}
  n11 = Custom!(Opacity=@MaterialExpressionTextureSampleParameter2D_0.G) {
   code="float mask = 0;\r\nif (Opacity >= 0.9)\r\n{\r\nclip(-1);\r\n}\r\nreturn mask;"
   outputType=CMOT_Float1
   inputs [{
     inputName=Opacity
     input {
      expression=&self:MaterialExpressionTextureSampleParameter2D_0
      mask=1
      maskG=1
      outputIndex=2
     }
    }]
  }
  Add(A=@n11)
  n12 = Multiply(
   A=DynamicParameter() {parameterIndex=2, defaultValue=<rgba 1 1 0 0>, paramNames [TintScale, "", "", ""]}.TintScale
   B=VectorParameter() {defaultValue=<rgba 1.2 1.121 1.0812 1>}.RGB
  )
  MaterialFunctionCall(Emissive=LinearInterpolate(B=LinearInterpolate(A=@n12, Alpha=@MaterialExpressionTextureSampleParameter2D_0.R), Alpha=OneMinus(None=@n3.MarkFlat)) {constA=1}) {materialFunction=&MF_FX_EmissiveAdjustments}
  MaterialFunctionCall!(CompressedQuaternion=@n3.CompressedQuaternion) {materialFunction=/Niagara/Functions/DataPacking/Niagara_DecompressQuaternion}
  ParticleColor()
 ]
}
